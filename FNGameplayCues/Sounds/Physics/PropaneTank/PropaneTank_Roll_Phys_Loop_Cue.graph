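# Looping physics sound cue for a propane tank rolling, routed through the impacts sound class with state-loop physics attenuation.
# A Random node picks one of two looping roll waves with equal weight.
# A continuous modulator at the root scales volume and pitch from a Speed parameter, so the sound follows how fast the tank rolls.
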SoundCue /FNGameplayCues/Sounds/Physics/PropaneTank/PropaneTank_Roll_Phys_Loop_Cue
root=ModulatorContinuous_0
nodes [
 {
  id=ModulatorContinuous_0
  children [Random_1]
  properties {
   pitchModulationParams {default=0, maxOutput=1.3, minOutput=0.6, parameterName=Speed}
   volumeModulationParams {default=0, parameterName=Speed}
  }
 }
 {id=Random_1, children [WavePlayer_7,WavePlayer_8], properties {weights [1,1]}}
 {id=WavePlayer_7, properties {bLooping=true, soundWaveAssetPtr=&/PropaneTank_Roll_Loop_01}}
 {id=WavePlayer_8, properties {bLooping=true, soundWaveAssetPtr=&/PropaneTank_Roll_Loop_02}}
]
properties {
 attenuationSettings=/Game/Sounds/Physics/Settings/Physics_StateLoop_Att
 maxDistance=6000
 pitchMultiplier=1.05
 soundClassObject=&Fort_Impacts
 volumeMultiplier=0.5
}
